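# Control Rig for the TimberStakeSoul face accessory mesh that procedurally corrects the rear and side-rear hair base bones as the head moves.
# Aim constraints first orient the four hair base bones relative to spine_05.
# Three spherical pose readers then measure head orientation against spine_05, each weighting a Modify Transforms offset on the hair bones.
# Two readers have debug drawing enabled, and two additional Modify Transforms nodes are left unconnected.
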ControlRigBlueprint /BRCosmetics/Characters/Player/Female/Medium/Bodies/F_MED_TimberStakeSoul/Meshes/Parts/F_MED_TimberStakeSoul_FaceAcc_CtrlRig
graphs [
 RigVMFunctionLibrary {links [], nodes []}
 RigVMModel {nodes [
   RigUnit_BeginExecution = &RigUnit_BeginExecution()
   &RigUnit_AimConstraintLocalSpaceOffset(
    Child=(Type=Bone,Name="dyn_hair_rear_base_l")
    Filter=(bX=True,bY=True,bZ=True)
    Parents=((Item=(Type=Bone,Name="spine_05"),Weight=1))
    Weight=1
    WorldUp=(Target=(X=0,Y=0,Z=1),Kind=Location,Space=(Type=Bone,Name="spine_05"))
    ExecutePin=@RigUnit_BeginExecution.ExecutePin
   )
   > &RigUnit_AimConstraintLocalSpaceOffset(
    Child=(Type=Bone,Name="dyn_hair_side_rear_base_l")
    Filter=(bX=True,bY=True,bZ=True)
    Parents=((Item=(Type=Bone,Name="spine_05"),Weight=1))
    Weight=1
    WorldUp=(Target=(X=0,Y=0,Z=1),Kind=Location,Space=(Type=Bone,Name="spine_05"))
   )
   > &RigUnit_AimConstraintLocalSpaceOffset(
    Child=(Type=Bone,Name="dyn_hair_side_rear_base_r")
    Filter=(bX=True,bY=True,bZ=True)
    Parents=((Item=(Type=Bone,Name="spine_05"),Weight=1))
    Weight=1
    WorldUp=(Target=(X=0,Y=0,Z=1),Kind=Location,Space=(Type=Bone,Name="spine_05"))
   )
   > &RigUnit_AimConstraintLocalSpaceOffset(
    Child=(Type=Bone,Name="dyn_hair_rear_base_r")
    Filter=(bX=True,bY=True,bZ=True)
    Parents=((Item=(Type=Bone,Name="spine_05"),Weight=1))
    Weight=1
    WorldUp=(Target=(X=0,Y=0,Z=1),Kind=Location,Space=(Type=Bone,Name="spine_05"))
   )
   > n1 = &RigUnit_SphericalPoseReader(
    ActiveRegionScaleFactors=(PositiveWidth=3.5,NegativeWidth=0.5,PositiveHeight=3.5,NegativeHeight=3.5)
    ActiveRegionSize=0.1
    Debug=(bDrawDebug=false,bDraw2D=False,bDrawLocalAxes=False,DebugScale=25,DebugSegments=20,DebugThickness=0.25)
    DriverAxis=(X=1,Y=0,Z=0)
    DriverItem=(Type=Bone,Name="head")
    FalloffRegionScaleFactors=(PositiveWidth=1,NegativeWidth=0.93,PositiveHeight=1,NegativeHeight=1)
    FalloffSize=0.35
    FlipHeightScaling=False
    FlipWidthScaling=False
    OptionalParentItem=(Type=Bone,Name="spine_05")
    RotationOffset=(X=90,Y=0,Z=-170)
   )
   > &RigUnit_ModifyTransforms(
    ItemToModify="((Item=(Type=Bone,Name=\"dyn_hair_side_rear_base_l\"),Transform=(Rotation=(X=0,Y=-0.1305,Z=0,W=0.9914),Translation=(X=-5,Y=0,Z=1),Scale3D=(X=1,Y=1,Z=1))),(Item=(Type=Bone,Name=\"dyn_hair_rear_base_l\"),Transform=(Rotation=(X=0,Y=-0.1305,Z=0,W=0.9914),Translati ...(truncated)"
    WeightMinimum=0
    Weight=@n1.OutputParam
   )
   > n2 = &RigUnit_SphericalPoseReader(
    ActiveRegionScaleFactors=(PositiveWidth=3.5,NegativeWidth=0.5,PositiveHeight=0.5,NegativeHeight=3.5)
    ActiveRegionSize=0.1
    Debug=(bDrawDebug=true,bDraw2D=False,bDrawLocalAxes=False,DebugScale=25,DebugSegments=20,DebugThickness=0.25)
    DriverAxis=(X=1,Y=0,Z=0)
    DriverItem=(Type=Bone,Name="head")
    FalloffRegionScaleFactors=(PositiveWidth=1,NegativeWidth=0.93,PositiveHeight=1,NegativeHeight=1)
    FalloffSize=0.35
    FlipHeightScaling=False
    FlipWidthScaling=False
    OptionalParentItem=(Type=Bone,Name="spine_05")
    RotationOffset=(X=70,Y=0,Z=-170)
   )
   > &RigUnit_ModifyTransforms(
    ItemToModify="((Item=(Type=Bone,Name=\"dyn_hair_side_rear_base_l\"),Transform=(Rotation=(X=0,Y=0,Z=0,W=1),Translation=(X=0,Y=0,Z=0),Scale3D=(X=1,Y=1,Z=1))),(Item=(Type=Bone,Name=\"dyn_hair_rear_base_l\"),Transform=(Rotation=(X=0,Y=0,Z=0,W=1),Translation=(X=0,Y=0,Z=0),Scale3 ...(truncated)"
    WeightMinimum=0
    Weight=@n2.OutputParam
   )
   > n3 = &RigUnit_SphericalPoseReader(
    ActiveRegionScaleFactors=(PositiveWidth=1,NegativeWidth=1,PositiveHeight=1,NegativeHeight=1)
    ActiveRegionSize=0.1
    Debug=(bDrawDebug=true,bDraw2D=False,bDrawLocalAxes=False,DebugScale=25,DebugSegments=20,DebugThickness=0.25)
    DriverAxis=(X=1,Y=0,Z=0)
    DriverItem=(Type=Bone,Name="head")
    FalloffRegionScaleFactors=(PositiveWidth=1,NegativeWidth=1,PositiveHeight=1,NegativeHeight=1)
    FalloffSize=0.2
    FlipHeightScaling=False
    FlipWidthScaling=False
    OptionalParentItem=(Type=Bone,Name="spine_05")
    RotationOffset=(X=50,Y=0,Z=-120)
   )
   > &RigUnit_ModifyTransforms(
    ItemToModify="((Item=(Type=Bone,Name=\"dyn_hair_side_rear_base_l\"),Transform=(Rotation=(X=0,Y=0,Z=0,W=1),Translation=(X=0,Y=0,Z=0),Scale3D=(X=1,Y=1,Z=1))),(Item=(Type=Bone,Name=\"dyn_hair_rear_base_l\"),Transform=(Rotation=(X=0,Y=0,Z=0,W=1),Translation=(X=0,Y=0,Z=0),Scale3 ...(truncated)"
    WeightMinimum=0
    Weight=@n3.OutputParam
   )
   &RigUnit_ModifyTransforms(
    ExecutePin=()
    ItemToModify="((Item=(Type=Bone,Name=\"dyn_hair_side_rear_base_l\"),Transform=(Rotation=(X=0.2271,Y=0.3122,Z=0.0744,W=-0.9195),Translation=(X=0,Y=0,Z=0),Scale3D=(X=1,Y=1,Z=1))),(Item=(Type=Bone,Name=\"dyn_hair_rear_base_l\"),Transform=(Rotation=(X=0,Y=0.3239,Z=0,W=-0.9461), ...(truncated)"
    Weight=1
    WeightMinimum=0
   )
   &RigUnit_ModifyTransforms(
    ExecutePin=()
    ItemToModify="((Item=(Type=Bone,Name=\"dyn_hair_side_rear_base_l\"),Transform=(Rotation=(X=0,Y=0,Z=0,W=1),Translation=(X=0,Y=0,Z=0),Scale3D=(X=1,Y=1,Z=1))),(Item=(Type=Bone,Name=\"dyn_hair_rear_base_l\"),Transform=(Rotation=(X=0,Y=0,Z=0,W=1),Translation=(X=0,Y=0,Z=0),Scale3 ...(truncated)"
    Weight=1
    WeightMinimum=0
   )
  ]}
]
hierarchy [
 root {loc=<0 0 0>, rot=<0 0 0 1>}
 pelvis {parent=root, loc=<0 -0.5301 98.0571>, rot=<0.0127 0.707 -0.0127 -0.707>}
 spine_01 {parent=pelvis, loc=<6.1208 -0.9328 0>, rot=<0 0 0.034 -0.9994>}
 spine_02 {parent=spine_01, loc=<6.5997 -0.2095 0>, rot=<0 0 0.0238 -0.9997>}
 spine_03 {parent=spine_02, loc=<6.4914 -0.1237 0>, rot=<0 0 0.0149 -0.9999>}
 spine_04 {parent=spine_03, loc=<6.9598 0.0794 0>, rot=<0 0 -0.0132 -0.9999>}
 spine_05 {parent=spine_04, loc=<6.9987 0 0>, rot=<0 0 -0.0015 -1>}
 neck_01 {parent=spine_05, loc=<11.3138 0.1529 0>, rot=<0 0 -0.2003 -0.9797>}
 neck_02 {parent=neck_01, loc=<4.5627 0 0>, rot=<0 0 0.0504 -0.9987>}
 head {parent=neck_02, loc=<5.3914 0 0>, rot=<0 0 0.115 -0.9934>}
 dyn_bangs_base {parent=head, loc=<12.2473 6.2501 0>, rot=<0.4979 0.5021 0.5021 0.4979>}
 dyn_bangs_01_l {parent=dyn_bangs_base, loc=<2.3194 -1.2803 -0.9016>, rot=<-0.0589 -0.407 0.2973 -0.8617>}
 dyn_bangs_02_l {parent=dyn_bangs_01_l, loc=<2.0513 0 0>, rot=<0.0475 -0.2494 0.1343 -0.9579>}
 dyn_bangs_03_l {parent=dyn_bangs_02_l, loc=<3.276 0 0>, rot=<0.0481 -0.1104 0.1304 -0.9841>}
 dyn_bangs_01_c {parent=dyn_bangs_base, loc=<2.8805 0.3385 -0.5808>, rot=<-0.07 -0.3941 0.029 -0.9159>}
 dyn_bangs_02_c {parent=dyn_bangs_01_c, loc=<2.0513 0 0>, rot=<-0.0308 -0.2376 0.1542 -0.9586>}
 dyn_bangs_03_c {parent=dyn_bangs_02_c, loc=<3.3078 0.1358 -0.0183>, rot=<0.0831 -0.1225 0.0529 -0.9876>}
 dyn_bangs_04_c {parent=dyn_bangs_02_c, loc=<2.8565 -0.5135 0.2027>, rot=<0.089 -0.1183 0.1009 -0.9838>}
 dyn_bangs_01_r {parent=dyn_bangs_base, loc=<2.2127 2.3747 -1.4528>, rot=<0.0423 -0.2656 -0.1706 -0.9479>}
 dyn_bangs_02_r {parent=dyn_bangs_01_r, loc=<1.9555 0.9643 -0.6694>, rot=<0.1272 -0.4075 -0.0154 -0.9042>}
 dyn_bangs_03_r {parent=dyn_bangs_02_r, loc=<2.5356 0.1858 -0.0283>, rot=<0.0646 0.0027 -0.0011 -0.9979>}
 dyn_bangs_04_r {parent=dyn_bangs_01_r, loc=<2.1734 -0.7091 -0.2265>, rot=<0.1539 -0.3651 0.015 -0.918>}
 dyn_bangs_05_r {parent=dyn_bangs_04_r, loc=<2.5356 0.1858 -0.0283>, rot=<0.0646 0.0024 -0.0055 -0.9979>}
 dyn_hair_side_rear_base_l {parent=head, loc=<12.3417 -2.2068 -6.0163>, rot=<0.1258 -0.9327 0.2723 -0.2004>}
 dyn_hair_side_rear_01_l {parent=dyn_hair_side_rear_base_l, loc=<0 0 0>, rot=<0 0 0 1>}
 dyn_hair_side_rear_02_l {parent=dyn_hair_side_rear_01_l, loc=<8.0276 0 0>, rot=<-0.0715 -0.3255 -0.0247 -0.9425>}
 dyn_hair_side_rear_03_l {parent=dyn_hair_side_rear_02_l, loc=<4.9704 0 0>, rot=<0 0 0 1>}
 dyn_hair_side_rear_04_l {parent=dyn_hair_side_rear_03_l, loc=<3.7727 0 0>, rot=<0 0 0 1>}
 dyn_hair_side_rear_05_l {parent=dyn_hair_side_rear_04_l, loc=<1.7658 2.6108 -3.5892>, rot=<0.4334 -0.0706 -0.3028 -0.8458>}
 dyn_hair_side_rear_06_l {parent=dyn_hair_side_rear_03_l, loc=<2.6001 2.5319 -3.2338>, rot=<0.4099 -0.1576 -0.4691 -0.7662>}
 dyn_hair_side_rear_07_l {parent=dyn_hair_side_rear_06_l, loc=<2.3648 -0.1836 0.4763>, rot=<-0.0227 0.1165 -0.0999 -0.9879>}
 dyn_hair_rear_base_l {parent=head, loc=<12.8907 -4.7992 -2.0407>, rot=<0.1658 -0.7916 0.5418 -0.2287>}
 dyn_hair_rear_01_l {parent=dyn_hair_rear_base_l, loc=<0 0 0>, rot=<0 0 0 1>}
 dyn_hair_rear_02_l {parent=dyn_hair_rear_01_l, loc=<7.8728 0 0>, rot=<-0.0327 -0.3262 -0.0113 -0.9447>}
 dyn_hair_rear_03_l {parent=dyn_hair_rear_02_l, loc=<4.9704 0 0>, rot=<0 0 0 1>}
 dyn_hair_rear_04_l {parent=dyn_hair_rear_03_l, loc=<3.7727 0 0>, rot=<0 0 0 1>}
 dyn_hair_side_rear_base_r {parent=head, loc=<12.3414 -2.2068 6.0163>, rot=<0.9327 0.1258 -0.2004 -0.2723>}
 dyn_hair_side_rear_01_r {parent=dyn_hair_side_rear_base_r, loc=<0 0 0>, rot=<0 0 0 1>}
 dyn_hair_side_rear_02_r {parent=dyn_hair_side_rear_01_r, loc=<-8.0273 0 -0.0002>, rot=<-0.0715 -0.3255 -0.0247 -0.9425>}
 dyn_hair_side_rear_03_r {parent=dyn_hair_side_rear_02_r, loc=<-4.9706 0 0>, rot=<0 0 0 1>}
 dyn_hair_side_rear_04_r {parent=dyn_hair_side_rear_03_r, loc=<-3.7731 0 -0.0001>, rot=<0 0 0 1>}
 dyn_hair_rear_base_r {parent=head, loc=<12.8909 -4.7992 2.0407>, rot=<0.7916 0.1658 -0.2287 -0.5418>}
 dyn_hair_rear_01_r {parent=dyn_hair_rear_base_r, loc=<0 0 0>, rot=<0 0 0 1>}
 dyn_hair_rear_02_r {parent=dyn_hair_rear_01_r, loc=<-7.8726 0 -0.0001>, rot=<-0.0327 -0.3262 -0.0113 -0.9447>}
 dyn_hair_rear_03_r {parent=dyn_hair_rear_02_r, loc=<-4.9707 0 0>, rot=<0 0 0 1>}
 dyn_hair_rear_04_r {parent=dyn_hair_rear_03_r, loc=<-3.7733 0 -0.0001>, rot=<0 0 0 1>}
 dyn_hair_side_rear_bounce_base_l {parent=head, loc=<5.2129 -4.9665 -8.4672>, rot=<-0.0074 -0.9279 0.3593 0.099>}
 dyn_hair_side_rear_bounce_l {parent=dyn_hair_side_rear_bounce_base_l, loc=<0 0 0>, rot=<0 0 0 1>}
 dyn_hair_rear_bounce_base_l {parent=head, loc=<6.2745 -8.8167 -3.4777>, rot=<-0.0366 -0.8066 0.5892 0.0306>}
 dyn_hair_rear_bounce_l {parent=dyn_hair_rear_bounce_base_l, loc=<0 0 0>, rot=<0 0 0 1>}
 dyn_hair_rear_bounce_base_r {parent=head, loc=<6.2749 -8.8168 3.4778>, rot=<0.8066 -0.0366 0.0306 -0.5892>}
 dyn_hair_rear_bounce_r {parent=dyn_hair_rear_bounce_base_r, loc=<0 0 0>, rot=<0 0 0 1>}
 dyn_hair_side_rear_bounce_base_r {parent=head, loc=<5.213 -4.9665 8.4673>, rot=<0.9279 -0.0074 0.099 -0.3593>}
 dyn_hair_side_rear_bounce_r {parent=dyn_hair_side_rear_bounce_base_r, loc=<0 0 0>, rot=<0 0 0 1>}
 deform_neck_no_twist {parent=spine_05, loc=<11.3138 0.1529 0>, rot=<0 0 -0.2003 -0.9797>}
 clavicle_l {parent=spine_05, loc=<5.4831 -0.1053 -3.2946>, rot=<0.0465 -0.7289 -0.009 -0.6829>}
 upperarm_l {parent=clavicle_l, loc=<9.8658 -1.2136 -0.4105>, rot=<0.0164 -0.3245 -0.0269 -0.9454>}
 clavicle_r {parent=spine_05, loc=<5.4831 -0.1053 3.2946>, rot=<0.7289 0.0465 -0.6829 0.009>}
 upperarm_r {parent=clavicle_r, loc=<-9.8658 1.2136 0.4105>, rot=<0.0164 -0.3245 -0.0269 -0.9454>}
]
