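# Physics asset for the TimberStakeSoul female medium FaceAcc mesh, giving its bangs and rear and side-rear hair physics bodies.
# Mostly capsule bodies form angle-limited bone chains from head-linked bases, alongside head, neck, upper spine, clavicle and upper-arm shapes.
# Free-rotating bounce helpers link their own head-linked bases to the third segment of each rear chain.
# Free-rotating cross constraints also tie the fourth rear and side-rear segments together across both sides.
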PhysicsAsset /BRCosmetics/Characters/Player/Female/Medium/Bodies/F_MED_TimberStakeSoul/Meshes/Parts/F_MED_TimberStakeSoul_FaceAcc_Physics
bodies {
 clavicle_l [{
   shapeType=Capsule
   length=9.2425
   radius=5.8875
   shapeName=clavicle_l_capsule
   center=<3.4257 -1.3409 -1.7391>
   rotation=<rot -0.0014 -89.9999 88.0173>
  }]
 clavicle_r [{
   shapeType=Capsule
   length=12.8931
   radius=6.7428
   shapeName=clavicle_r_capsule
   center=<-3.4257 1.3409 1.7391>
   rotation=<rot -0.0015 -89.9999 -77.829>
  }]
 dyn_bangs_01_c [{shapeType=Capsule, length=1.5194, radius=1.2837, shapeName=dyn_bangs_01_c_capsule, center=<2 0 0>, rotation=<rot 90 0 0>}]
 dyn_bangs_01_l [{shapeType=Capsule, length=3.1119, radius=1.4813, shapeName=dyn_bangs_01_l_capsule, center=<2 0 0>, rotation=<rot 90 0 0>}]
 dyn_bangs_01_r [{shapeType=Capsule, length=1.8694, radius=1.0857, shapeName=dyn_bangs_01_r_capsule, center=<2 0 0>, rotation=<rot 90 0 0>}]
 dyn_bangs_02_c [{shapeType=Capsule, length=1.9147, radius=0.7309, shapeName=dyn_bangs_02_c_capsule, center=<2 0 0>, rotation=<rot 90 0 0>}]
 dyn_bangs_02_l [{shapeType=Capsule, length=2.2025, radius=1.9011, shapeName=dyn_bangs_02_l_capsule, center=<2 0 0>, rotation=<rot 90 0 0>}]
 dyn_bangs_02_r [{shapeType=Capsule, length=0.505, radius=0.505, shapeName=dyn_bangs_02_r_capsule, center=<2 0 0>, rotation=<rot 90 0 0>}]
 dyn_bangs_03_c [{shapeType=Capsule, length=1.75, radius=0.88, shapeName=dyn_bangs_03_c_capsule, center=<2 0 0>, rotation=<rot 53.938 0 0>}]
 dyn_bangs_03_l [{shapeType=Capsule, length=1.9671, radius=1.3854, shapeName=dyn_bangs_03_l_capsule, center=<2 0 0>, rotation=<rot 61.4833 24.1191 21.4744>}]
 dyn_bangs_03_r [{shapeType=Capsule, length=1.7864, radius=1.1529, shapeName=dyn_bangs_03_r_capsule, center=<2 0 0>, rotation=<rot 34.9819 73.5142 34.4285>}]
 dyn_bangs_04_c [{shapeType=Capsule, length=1.95, radius=1.305, shapeName=dyn_bangs_04_c_capsule, center=<2 0 0>, rotation=<rot 61.4417 -0.0001 -0.0001>}]
 dyn_bangs_04_r [{shapeType=Capsule, length=1.4391, radius=1.3623, shapeName=dyn_bangs_04_r_capsule, center=<2 0 0>, rotation=<rot 90 0 0>}]
 dyn_bangs_05_r [{shapeType=Capsule, length=2.1181, radius=1.281, shapeName=dyn_bangs_05_r_capsule, center=<2 0 0>, rotation=<rot 15.2895 83.8913 16.6748>}]
 dyn_bangs_base [{shapeType=Capsule, length=6, radius=1, rotation=<rot 0 0 90>}]
 dyn_hair_rear_01_l [{shapeType=Capsule, length=4.7812, radius=1.5625, shapeName=dyn_hair_rear_01_l_capsule, center=<3 0 2.0441>, rotation=<rot 90 -90 0>}]
 dyn_hair_rear_01_r [{shapeType=Capsule, length=4.7812, radius=1.5625, shapeName=dyn_hair_rear_01_r_capsule, center=<-3 0 -2.0441>, rotation=<rot 90 90 0>}]
 dyn_hair_rear_02_l [{shapeType=Capsule, length=4.0038, radius=2.7886, shapeName=dyn_hair_rear_02_l_capsule, center=<3 0 1.694>, rotation=<rot 90 -90 0>}]
 dyn_hair_rear_02_r [{shapeType=Capsule, length=4.0038, radius=2.7886, shapeName=dyn_hair_rear_02_r_capsule, center=<-3 0 -1.694>, rotation=<rot 90 90 0>}]
 dyn_hair_rear_03_l [{shapeType=Capsule, length=4.7188, radius=2.7283, shapeName=dyn_hair_rear_03_l_capsule, center=<3 0 1.3879>, rotation=<rot 90 -90 0>}]
 dyn_hair_rear_03_r [{shapeType=Capsule, length=4.7188, radius=2.7283, shapeName=dyn_hair_rear_03_r_capsule, center=<-3 0 -1.3879>, rotation=<rot 90 90 0>}]
 dyn_hair_rear_04_l [{shapeType=Capsule, length=7.4423, radius=4.3513, shapeName=dyn_hair_rear_04_l_capsule, center=<3 0 2.3342>, rotation=<rot 90 -90 0>}]
 dyn_hair_rear_04_r [{shapeType=Capsule, length=9.7861, radius=4.1326, shapeName=dyn_hair_rear_04_r_capsule, center=<-3 0 -2.0861>, rotation=<rot 90 90 0>}]
 dyn_hair_rear_base_l [{shapeType=Capsule, length=0.505, radius=0.505, shapeName=dyn_hair_rear_base_l_capsule, rotation=<rot 0 -90 0>}]
 dyn_hair_rear_base_r [{shapeType=Capsule, length=0.505, radius=0.505, shapeName=dyn_hair_rear_base_r_capsule, rotation=<rot 0 -90 0>}]
 dyn_hair_rear_bounce_base_l [{shapeType=Capsule, length=0.505, radius=0.505, shapeName=dyn_hair_rear_bounce_base_l_capsule, rotation=<rot 0 -90 0>}]
 dyn_hair_rear_bounce_base_r [{shapeType=Capsule, length=0.505, radius=0.505, shapeName=dyn_hair_rear_bounce_base_r_capsule, rotation=<rot 0 -90 0>}]
 dyn_hair_rear_bounce_l [{shapeType=Capsule, length=0.505, radius=0.505, shapeName=dyn_hair_rear_bounce_l_capsule, rotation=<rot 0 -90 0>}]
 dyn_hair_rear_bounce_r [{shapeType=Capsule, length=0.505, radius=0.505, shapeName=dyn_hair_rear_bounce_r_capsule, rotation=<rot 0 -90 0>}]
 dyn_hair_side_rear_01_l [{shapeType=Capsule, length=3.3655, radius=1.5434, shapeName=dyn_hair_side_rear_01_l_capsule, center=<2 0 1.6718>, rotation=<rot 90 -90 0>}]
 dyn_hair_side_rear_01_r [{shapeType=Capsule, length=3.3655, radius=1.5434, shapeName=dyn_hair_side_rear_01_r_capsule, center=<-2 0 -1.6718>, rotation=<rot 90 90 0>}]
 dyn_hair_side_rear_02_l [{shapeType=Capsule, length=5.0832, radius=3.0924, shapeName=dyn_hair_side_rear_02_l_capsule, center=<2 0 2>, rotation=<rot 90 -90 0>}]
 dyn_hair_side_rear_02_r [{shapeType=Capsule, length=5.0832, radius=3.0924, shapeName=dyn_hair_side_rear_02_r_capsule, center=<-2 0 -2>, rotation=<rot 90 90 0>}]
 dyn_hair_side_rear_03_l [{shapeType=Capsule, length=7.3741, radius=2.903, shapeName=dyn_hair_side_rear_03_l_capsule, center=<2 0 1.9373>, rotation=<rot 90 -90 0>}]
 dyn_hair_side_rear_03_r [{shapeType=Capsule, length=7.3741, radius=2.903, shapeName=dyn_hair_side_rear_03_r_capsule, center=<-2 0 -1.9373>, rotation=<rot 90 90 0>}]
 dyn_hair_side_rear_04_l [{shapeType=Capsule, length=6.5703, radius=3.4232, shapeName=dyn_hair_side_rear_04_l_capsule, center=<2 0 1.7547>, rotation=<rot 90 -90 0>}]
 dyn_hair_side_rear_04_r [{
   shapeType=Capsule
   length=5.3871
   radius=4.6065
   shapeName=dyn_hair_side_rear_04_r_capsule
   center=<-1.6749 0.0929 -3.3455>
   rotation=<rot 63.3448 116.5317 11.8126>
  }]
 dyn_hair_side_rear_05_l [{
   shapeType=Capsule
   length=3.0625
   radius=1.2812
   shapeName=dyn_hair_side_rear_05_l_capsule
   center=<1 0 0.8941>
   rotation=<rot -0.0012 102.2178 80.322>
  }]
 dyn_hair_side_rear_06_l [{
   shapeType=Capsule
   length=1.4438
   radius=0.9463
   shapeName=dyn_hair_side_rear_06_l_capsule
   center=<1 0 1.6943>
   rotation=<rot 0 90.0001 72.8132>
  }]
 dyn_hair_side_rear_07_l [{shapeType=Capsule, length=0.505, radius=0.505, shapeName=dyn_hair_side_rear_07_l_capsule, center=<1 0 0>, rotation=<rot 0 90 90>}]
 dyn_hair_side_rear_base_l [{shapeType=Sphere, radius=0.505, shapeName=dyn_hair_side_rear_base_l_sphere}]
 dyn_hair_side_rear_base_r [{shapeType=Capsule, length=0.505, radius=0.505, shapeName=dyn_hair_side_rear_base_r_capsule, rotation=<rot 0 -90 0>}]
 dyn_hair_side_rear_bounce_base_l [{shapeType=Capsule, length=0.505, radius=0.505, shapeName=dyn_hair_side_rear_bounce_base_l_capsule, rotation=<rot 0 -90 0>}]
 dyn_hair_side_rear_bounce_base_r [{shapeType=Capsule, length=0.505, radius=0.505, shapeName=dyn_hair_side_rear_bounce_base_r_capsule, rotation=<rot 0 -90 0>}]
 dyn_hair_side_rear_bounce_l [{shapeType=Capsule, length=0.505, radius=0.505, shapeName=dyn_hair_side_rear_bounce_l_capsule, rotation=<rot 0 -90 0>}]
 dyn_hair_side_rear_bounce_r [{shapeType=Capsule, length=0.505, radius=0.505, shapeName=dyn_hair_side_rear_bounce_r_capsule, rotation=<rot 0 -90 0>}]
 head [{shapeType=Capsule, length=4.8958, radius=8.0467, center=<4.5528 0 0>, rotation=<rot 90 0 0>}]
 neck_01 [{shapeType=Capsule, length=5.5391, radius=4.4659, shapeName=neck_01_capsule, rotation=<rot 90 0 0>}]
 spine_05 [{shapeType=Capsule, length=13.6024, radius=10.0727, shapeName=spine_05_capsule, center=<0.8592 -0.8043 0>, rotation=<rot 0 -90 0>}]
 upperarm_l [{
   shapeType=Capsule
   length=16.9399
   radius=5.6326
   shapeName=upperarm_l_capsule
   center=<8.2316 0.0034 -1.0438>
   rotation=<rot -0.023 -89.9795 87.5163>
  }]
 upperarm_r [{
   shapeType=Capsule
   length=16.9399
   radius=5.6326
   shapeName=upperarm_r_capsule
   center=<-8.2316 -0.0034 1.0438>
   rotation=<rot -0.023 -89.9795 -92.4837>
  }]
}
constraints [
 {bone1Name=dyn_bangs_base, bone2Name=head, swing1LimitDegrees=45, swing2LimitDegrees=45, twistLimitDegrees=45}
 {bone1Name=dyn_bangs_01_c, bone2Name=dyn_bangs_base, swing1LimitDegrees=6, swing2LimitDegrees=6, twistLimitDegrees=6}
 {bone1Name=dyn_bangs_02_c, bone2Name=dyn_bangs_01_c, swing1LimitDegrees=6, swing2LimitDegrees=6, twistLimitDegrees=6}
 {bone1Name=dyn_bangs_03_c, bone2Name=dyn_bangs_02_c, swing1LimitDegrees=28, swing2LimitDegrees=28, twistLimitDegrees=14}
 {bone1Name=dyn_bangs_04_c, bone2Name=dyn_bangs_02_c, swing1LimitDegrees=28, swing2LimitDegrees=28, twistLimitDegrees=14}
 {bone1Name=dyn_bangs_01_l, bone2Name=dyn_bangs_base, swing1LimitDegrees=6, swing2LimitDegrees=6, twistLimitDegrees=6}
 {bone1Name=dyn_bangs_02_l, bone2Name=dyn_bangs_01_l, swing1LimitDegrees=14, swing2LimitDegrees=14, twistLimitDegrees=14}
 {bone1Name=dyn_bangs_03_l, bone2Name=dyn_bangs_02_l, swing1LimitDegrees=28, swing2LimitDegrees=28, twistLimitDegrees=14}
 {bone1Name=dyn_bangs_01_r, bone2Name=dyn_bangs_base, swing1LimitDegrees=6, swing2LimitDegrees=6, twistLimitDegrees=6}
 {bone1Name=dyn_bangs_02_r, bone2Name=dyn_bangs_01_r, swing1LimitDegrees=14, swing2LimitDegrees=14, twistLimitDegrees=14}
 {bone1Name=dyn_bangs_03_r, bone2Name=dyn_bangs_02_r, swing1LimitDegrees=28, swing2LimitDegrees=28, twistLimitDegrees=14}
 {bone1Name=dyn_bangs_04_r, bone2Name=dyn_bangs_01_r, swing1LimitDegrees=14, swing2LimitDegrees=14, twistLimitDegrees=14}
 {bone1Name=dyn_bangs_05_r, bone2Name=dyn_bangs_04_r, swing1LimitDegrees=28, swing2LimitDegrees=28, twistLimitDegrees=14}
 {
  bone1Name=dyn_hair_rear_base_l
  bone2Name=head
  swing1LimitDegrees=45
  swing1Motion=Free
  swing2LimitDegrees=45
  swing2Motion=Free
  twistLimitDegrees=45
  twistMotion=Free
 }
 {bone1Name=dyn_hair_rear_01_l, bone2Name=dyn_hair_rear_base_l, swing1LimitDegrees=20, swing2LimitDegrees=20, twistLimitDegrees=20}
 {bone1Name=dyn_hair_rear_02_l, bone2Name=dyn_hair_rear_01_l, swing1LimitDegrees=8, swing2LimitDegrees=8, twistLimitDegrees=8}
 {bone1Name=dyn_hair_rear_03_l, bone2Name=dyn_hair_rear_02_l, swing1LimitDegrees=8, swing2LimitDegrees=8, twistLimitDegrees=8}
 {bone1Name=dyn_hair_rear_04_l, bone2Name=dyn_hair_rear_03_l, swing1LimitDegrees=8, swing2LimitDegrees=8, twistLimitDegrees=8}
 {
  bone1Name=dyn_hair_rear_base_r
  bone2Name=head
  swing1LimitDegrees=45
  swing1Motion=Free
  swing2LimitDegrees=45
  swing2Motion=Free
  twistLimitDegrees=45
  twistMotion=Free
 }
 {bone1Name=dyn_hair_rear_01_r, bone2Name=dyn_hair_rear_base_r, swing1LimitDegrees=20, swing2LimitDegrees=20, twistLimitDegrees=20}
 {bone1Name=dyn_hair_rear_02_r, bone2Name=dyn_hair_rear_01_r, swing1LimitDegrees=8, swing2LimitDegrees=8, twistLimitDegrees=8}
 {bone1Name=dyn_hair_rear_03_r, bone2Name=dyn_hair_rear_02_r, swing1LimitDegrees=8, swing2LimitDegrees=8, twistLimitDegrees=8}
 {bone1Name=dyn_hair_rear_04_r, bone2Name=dyn_hair_rear_03_r, swing1LimitDegrees=8, swing2LimitDegrees=8, twistLimitDegrees=8}
 {bone1Name=dyn_hair_side_rear_01_l, bone2Name=dyn_hair_side_rear_base_l, swing1LimitDegrees=20, swing2LimitDegrees=20, twistLimitDegrees=20}
 {bone1Name=dyn_hair_side_rear_02_l, bone2Name=dyn_hair_side_rear_01_l, swing1LimitDegrees=8, swing2LimitDegrees=8, twistLimitDegrees=8}
 {bone1Name=dyn_hair_side_rear_03_l, bone2Name=dyn_hair_side_rear_02_l, swing1LimitDegrees=8, swing2LimitDegrees=8, twistLimitDegrees=8}
 {bone1Name=dyn_hair_side_rear_04_l, bone2Name=dyn_hair_side_rear_03_l, swing1LimitDegrees=8, swing2LimitDegrees=8, twistLimitDegrees=8}
 {
  bone1Name=dyn_hair_side_rear_base_r
  bone2Name=head
  swing1LimitDegrees=45
  swing1Motion=Free
  swing2LimitDegrees=45
  swing2Motion=Free
  twistLimitDegrees=45
  twistMotion=Free
 }
 {bone1Name=dyn_hair_side_rear_01_r, bone2Name=dyn_hair_side_rear_base_r, swing1LimitDegrees=20, swing2LimitDegrees=20, twistLimitDegrees=20}
 {bone1Name=dyn_hair_side_rear_02_r, bone2Name=dyn_hair_side_rear_01_r, swing1LimitDegrees=8, swing2LimitDegrees=8, twistLimitDegrees=8}
 {bone1Name=dyn_hair_side_rear_03_r, bone2Name=dyn_hair_side_rear_02_r, swing1LimitDegrees=8, swing2LimitDegrees=8, twistLimitDegrees=8}
 {bone1Name=dyn_hair_side_rear_04_r, bone2Name=dyn_hair_side_rear_03_r, swing1LimitDegrees=8, swing2LimitDegrees=8, twistLimitDegrees=8}
 {
  bone1Name=dyn_hair_rear_bounce_l
  bone2Name=dyn_hair_rear_bounce_base_l
  swing1LimitDegrees=45
  swing1Motion=Free
  swing2LimitDegrees=45
  swing2Motion=Free
  twistLimitDegrees=45
  twistMotion=Free
 }
 {
  bone1Name=dyn_hair_rear_bounce_r
  bone2Name=dyn_hair_rear_bounce_base_r
  swing1LimitDegrees=45
  swing1Motion=Free
  swing2LimitDegrees=45
  swing2Motion=Free
  twistLimitDegrees=45
  twistMotion=Free
 }
 {
  bone1Name=dyn_hair_side_rear_bounce_l
  bone2Name=dyn_hair_side_rear_bounce_base_l
  swing1LimitDegrees=45
  swing1Motion=Free
  swing2LimitDegrees=45
  swing2Motion=Free
  twistLimitDegrees=45
  twistMotion=Free
 }
 {
  bone1Name=dyn_hair_side_rear_bounce_r
  bone2Name=dyn_hair_side_rear_bounce_base_r
  swing1LimitDegrees=45
  swing1Motion=Free
  swing2LimitDegrees=45
  swing2Motion=Free
  twistLimitDegrees=45
  twistMotion=Free
 }
 {
  bone1Name=dyn_hair_side_rear_bounce_base_r
  bone2Name=head
  swing1LimitDegrees=45
  swing1Motion=Free
  swing2LimitDegrees=45
  swing2Motion=Free
  twistLimitDegrees=45
  twistMotion=Free
 }
 {
  bone1Name=dyn_hair_side_rear_bounce_base_l
  bone2Name=head
  swing1LimitDegrees=45
  swing1Motion=Free
  swing2LimitDegrees=45
  swing2Motion=Free
  twistLimitDegrees=45
  twistMotion=Free
 }
 {
  bone1Name=dyn_hair_rear_bounce_base_r
  bone2Name=head
  swing1LimitDegrees=45
  swing1Motion=Free
  swing2LimitDegrees=45
  swing2Motion=Free
  twistLimitDegrees=45
  twistMotion=Free
 }
 {
  bone1Name=dyn_hair_rear_bounce_base_l
  bone2Name=head
  swing1LimitDegrees=45
  swing1Motion=Free
  swing2LimitDegrees=45
  swing2Motion=Free
  twistLimitDegrees=45
  twistMotion=Free
 }
 {
  bone1Name=dyn_hair_side_rear_base_l
  bone2Name=head
  swing1LimitDegrees=45
  swing1Motion=Free
  swing2LimitDegrees=45
  swing2Motion=Free
  twistLimitDegrees=45
  twistMotion=Free
 }
 {
  bone1Name=dyn_hair_rear_bounce_l
  bone2Name=dyn_hair_rear_03_l
  swing1LimitDegrees=45
  swing1Motion=Free
  swing2LimitDegrees=45
  swing2Motion=Free
  twistLimitDegrees=45
  twistMotion=Free
 }
 {
  bone1Name=dyn_hair_rear_bounce_r
  bone2Name=dyn_hair_rear_03_r
  swing1LimitDegrees=45
  swing1Motion=Free
  swing2LimitDegrees=45
  swing2Motion=Free
  twistLimitDegrees=45
  twistMotion=Free
 }
 {
  bone1Name=dyn_hair_side_rear_bounce_l
  bone2Name=dyn_hair_side_rear_03_l
  swing1LimitDegrees=45
  swing1Motion=Free
  swing2LimitDegrees=45
  swing2Motion=Free
  twistLimitDegrees=45
  twistMotion=Free
 }
 {
  bone1Name=dyn_hair_side_rear_bounce_r
  bone2Name=dyn_hair_side_rear_03_r
  swing1LimitDegrees=45
  swing1Motion=Free
  swing2LimitDegrees=45
  swing2Motion=Free
  twistLimitDegrees=45
  twistMotion=Free
 }
 {
  bone1Name=dyn_hair_side_rear_04_l
  bone2Name=dyn_hair_rear_04_l
  swing1LimitDegrees=45
  swing1Motion=Free
  swing2LimitDegrees=45
  swing2Motion=Free
  twistLimitDegrees=45
  twistMotion=Free
 }
 {
  bone1Name=dyn_hair_rear_04_r
  bone2Name=dyn_hair_rear_04_l
  swing1LimitDegrees=45
  swing1Motion=Free
  swing2LimitDegrees=45
  swing2Motion=Free
  twistLimitDegrees=45
  twistMotion=Free
 }
 {
  bone1Name=dyn_hair_rear_04_l
  bone2Name=dyn_hair_rear_04_r
  swing1LimitDegrees=45
  swing1Motion=Free
  swing2LimitDegrees=45
  swing2Motion=Free
  twistLimitDegrees=45
  twistMotion=Free
 }
 {
  bone1Name=neck_01
  bone2Name=spine_05
  swing1LimitDegrees=45
  swing1Motion=Free
  swing2LimitDegrees=45
  swing2Motion=Free
  twistLimitDegrees=45
  twistMotion=Free
 }
 {
  bone1Name=head
  bone2Name=spine_05
  swing1LimitDegrees=45
  swing1Motion=Free
  swing2LimitDegrees=45
  swing2Motion=Free
  twistLimitDegrees=45
  twistMotion=Free
 }
 {
  bone1Name=clavicle_l
  bone2Name=spine_05
  swing1LimitDegrees=45
  swing1Motion=Free
  swing2LimitDegrees=45
  swing2Motion=Free
  twistLimitDegrees=45
  twistMotion=Free
 }
 {
  bone1Name=upperarm_l
  bone2Name=clavicle_l
  swing1LimitDegrees=45
  swing1Motion=Free
  swing2LimitDegrees=45
  swing2Motion=Free
  twistLimitDegrees=45
  twistMotion=Free
 }
 {
  bone1Name=clavicle_r
  bone2Name=spine_05
  swing1LimitDegrees=45
  swing1Motion=Free
  swing2LimitDegrees=45
  swing2Motion=Free
  twistLimitDegrees=45
  twistMotion=Free
 }
 {
  bone1Name=upperarm_r
  bone2Name=clavicle_r
  swing1LimitDegrees=45
  swing1Motion=Free
  swing2LimitDegrees=45
  swing2Motion=Free
  twistLimitDegrees=45
  twistMotion=Free
 }
 {
  bone1Name=dyn_hair_rear_04_r
  bone2Name=dyn_hair_side_rear_04_r
  swing1LimitDegrees=45
  swing1Motion=Free
  swing2LimitDegrees=45
  swing2Motion=Free
  twistLimitDegrees=45
  twistMotion=Free
 }
 {bone1Name=dyn_hair_side_rear_05_l, bone2Name=dyn_hair_side_rear_04_l, swing1LimitDegrees=8, swing2LimitDegrees=8, twistLimitDegrees=8}
 {bone1Name=dyn_hair_side_rear_06_l, bone2Name=dyn_hair_side_rear_03_l, swing1LimitDegrees=8, swing2LimitDegrees=8, twistLimitDegrees=8}
 {bone1Name=dyn_hair_side_rear_07_l, bone2Name=dyn_hair_side_rear_06_l, swing1LimitDegrees=8, swing2LimitDegrees=8, twistLimitDegrees=8}
]
