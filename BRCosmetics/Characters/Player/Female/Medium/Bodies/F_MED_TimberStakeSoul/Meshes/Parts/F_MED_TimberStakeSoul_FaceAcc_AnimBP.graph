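# Animation Blueprint for the TimberStakeSoul female face accessory part, adding secondary motion to its hair.
# It copies the parent mesh pose through a Control Rig, then chains four linear-spring AnimDynamics nodes for rear hair bounce.
# A head-space RigidBody simulation follows, then additive bone transforms adjust side-rear hair and bangs.
# Each update sets a Gravity Override variable from a state-based parameter asset, though no AnimGraph node visibly reads it.
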AnimBlueprint /BRCosmetics/Characters/Player/Female/Medium/Bodies/F_MED_TimberStakeSoul/Meshes/Parts/F_MED_TimberStakeSoul_FaceAcc_AnimBP
target_skeleton=&/F_MED_TimberStakeSoul_FaceAcc_Skeleton
graphs [
 AnimGraph {nodes [
   n1 = Control Rig(Source=Copy Pose From Mesh() {bUseAttachedParent=true}) {
    alphaBoolBlend {blendOption=Linear}
    controlRigAssetReference {blueprintRigClass=&/F_MED_TimberStakeSoul_FaceAcc_CtrlRig.F_MED_TimberStakeSoul_FaceAcc_CtrlRig_C}
   }
   n2 = Anim Dynamics - Bone: dyn_hair_side_rear_bounce_l(ComponentPose=Local To Component(LocalPose=@n1), Alpha=0.9, ExternalForce="0,0,0") {
    bLinearSpring=true
    bOverrideLinearDamping=true
    bUseGravityOverride=true
    linearDampingOverride=0.999
    linearSpringConstant=75
    alphaBoolBlend {blendOption=Linear}
    boundBone {boneName=dyn_hair_side_rear_bounce_l}
    componentAppliedLinearAccClamp=<770 770 770>
    componentLinearAccScale=<0.2 0.2 0.2>
    componentLinearVelScale=<0.2 0.2 0.2>
    gravityOverride {z=-650}
    physicsBodyDefinitions [{
      boundBone {boneName=dyn_hair_side_rear_bounce_l}
      boxExtents=<60 60 60>
      constraintSetup {angularLimitsMax=<0 0 0>, angularLimitsMin=<0 0 0>, angularTarget=<0 0 0>, linearAxesMax=<0.5 1 1>, linearAxesMin=<-4.5 -1 -1>}
      localJointOffset=<4 0 0>
     }]
   }
   n3 = Anim Dynamics - Bone: dyn_hair_rear_bounce_l(ComponentPose=@n2, Alpha=0.9, ExternalForce="0,0,0") {
    bLinearSpring=true
    bOverrideLinearDamping=true
    bUseGravityOverride=true
    linearDampingOverride=0.999
    linearSpringConstant=75
    alphaBoolBlend {blendOption=Linear}
    boundBone {boneName=dyn_hair_rear_bounce_l}
    componentAppliedLinearAccClamp=<770 770 770>
    componentLinearAccScale=<0.2 0.2 0.2>
    componentLinearVelScale=<0.2 0.2 0.2>
    gravityOverride {z=-650}
    physicsBodyDefinitions [{
      boundBone {boneName=dyn_hair_rear_bounce_l}
      boxExtents=<60 60 60>
      constraintSetup {angularLimitsMax=<0 0 0>, angularLimitsMin=<0 0 0>, angularTarget=<0 0 0>, linearAxesMax=<0.5 1 1>, linearAxesMin=<-4.5 -1 -1>}
      localJointOffset=<4 0 0>
     }]
   }
   n4 = Anim Dynamics - Bone: dyn_hair_rear_bounce_r(ComponentPose=@n3, Alpha=0.9, ExternalForce="0,0,0") {
    bLinearSpring=true
    bOverrideLinearDamping=true
    bUseGravityOverride=true
    linearDampingOverride=0.999
    linearSpringConstant=75
    alphaBoolBlend {blendOption=Linear}
    boundBone {boneName=dyn_hair_rear_bounce_r}
    componentAppliedLinearAccClamp=<770 770 770>
    componentLinearAccScale=<0.2 0.2 0.2>
    componentLinearVelScale=<0.2 0.2 0.2>
    gravityOverride {z=-650}
    physicsBodyDefinitions [{
      boundBone {boneName=dyn_hair_rear_bounce_r}
      boxExtents=<60 60 60>
      constraintSetup {angularLimitsMax=<0 0 0>, angularLimitsMin=<0 0 0>, angularTarget=<0 0 0>, linearAxesMax=<4.5 1 1>, linearAxesMin=<-0.5 -1 -1>}
      localJointOffset=<-4 0 0>
     }]
   }
   n5 = Anim Dynamics - Bone: dyn_hair_side_rear_bounce_r(ComponentPose=@n4, Alpha=0.9, ExternalForce="0,0,0") {
    bLinearSpring=true
    bOverrideLinearDamping=true
    bUseGravityOverride=true
    linearDampingOverride=0.999
    linearSpringConstant=75
    alphaBoolBlend {blendOption=Linear}
    boundBone {boneName=dyn_hair_side_rear_bounce_r}
    componentAppliedLinearAccClamp=<770 770 770>
    componentLinearAccScale=<0.2 0.2 0.2>
    componentLinearVelScale=<0.2 0.2 0.2>
    gravityOverride {z=-650}
    physicsBodyDefinitions [{
      boundBone {boneName=dyn_hair_side_rear_bounce_r}
      boxExtents=<60 60 60>
      constraintSetup {angularLimitsMax=<0 0 0>, angularLimitsMin=<0 0 0>, angularTarget=<0 0 0>, linearAxesMax=<4.5 1 1>, linearAxesMin=<-0.5 -1 -1>}
      localJointOffset=<-4 0 0>
     }]
   }
   n6 = RigidBody!(ComponentPose=@n5, Alpha, OverrideWorldGravity:Vector="0,0,-980", ExternalForce) -> (Pose) {
    bOverrideWorldGravity=true
    overridePhysicsAsset=&/F_MED_TimberStakeSoul_FaceAcc_Physics
    simulationSpace=BaseBoneSpace
    alphaBoolBlend {blendOption=Linear}
    baseBoneRef {boneName=head}
    componentAppliedLinearAccClamp=<0 0 0>
    overrideWorldGravity {z=-980}
    simSpaceSettings {
     dampingAlpha=0
     maxAngularAcceleration=15
     maxAngularVelocity=3
     maxLinearAcceleration=2000
     maxLinearVelocity=300
     velocityScaleZ=0.3
     worldAlpha=0.95
     externalLinearDragV {z=0.1}
    }
   }
   n7 = "Transform (Modify) Bone - Bone: dyn_hair_side_rear_03_r"!(ComponentPose=@n6, bAlphaBoolEnabled:bool=True, Translation="0,0,-0.5", Rotation="10,0,0", Scale="0.8,0.8,0.8") -> (Pose) {
    alphaInputType=Bool
    rotationMode=BMM_Additive
    rotationSpace=BCS_BoneSpace
    scaleMode=BMM_Additive
    scaleSpace=BCS_BoneSpace
    translationMode=BMM_Additive
    translationSpace=BCS_BoneSpace
    alphaBoolBlend {blendOption=Linear}
    boneToModify {boneName=dyn_hair_side_rear_03_r}
    rotation {pitch=10}
    scale=<0.8 0.8 0.8>
    translation {z=-0.5}
   }
   n8 = "Transform (Modify) Bone - Bone: dyn_hair_side_rear_03_l"!(ComponentPose=@n7, bAlphaBoolEnabled:bool=True, Translation="0,0,0.5", Rotation="5,0,0", Scale="0.8,0.8,0.8") -> (Pose) {
    alphaInputType=Bool
    rotationMode=BMM_Additive
    rotationSpace=BCS_BoneSpace
    scaleMode=BMM_Additive
    scaleSpace=BCS_BoneSpace
    translationMode=BMM_Additive
    translationSpace=BCS_BoneSpace
    alphaBoolBlend {blendOption=Linear}
    boneToModify {boneName=dyn_hair_side_rear_03_l}
    rotation {pitch=5}
    scale=<0.8 0.8 0.8>
    translation {z=0.5}
   }
   n9 = "Transform (Modify) Bone - Bone: dyn_bangs_02_l"(ComponentPose=@n8, Alpha=1, Translation="0,0,0", Rotation="8.442,0.000,-0.000", Scale="1,1,1") {
    rotationMode=BMM_Additive
    rotationSpace=BCS_BoneSpace
    scaleMode=BMM_Additive
    scaleSpace=BCS_BoneSpace
    translationMode=BMM_Additive
    translationSpace=BCS_BoneSpace
    alphaBoolBlend {blendOption=Linear}
    boneToModify {boneName=dyn_bangs_02_l}
    rotation=<rot 8.4421 0 0>
   }
   n10 = "Transform (Modify) Bone - Bone: dyn_bangs_02_r"(ComponentPose=@n9, Alpha=1, Translation="0,0,0", Rotation="10,0,0", Scale="1,1,1") {
    rotationMode=BMM_Additive
    rotationSpace=BCS_BoneSpace
    scaleMode=BMM_Additive
    scaleSpace=BCS_BoneSpace
    translationMode=BMM_Additive
    translationSpace=BCS_BoneSpace
    alphaBoolBlend {blendOption=Linear}
    boneToModify {boneName=dyn_bangs_02_r}
    rotation {pitch=10}
   }
   n11 = "Transform (Modify) Bone - Bone: dyn_bangs_02_c"(ComponentPose=@n10, Alpha=1, Translation="0,0,0", Rotation="8.442,0.000,-0.000", Scale="1,1,1") {
    rotationMode=BMM_Additive
    rotationSpace=BCS_BoneSpace
    scaleMode=BMM_Additive
    scaleSpace=BCS_BoneSpace
    translationMode=BMM_Additive
    translationSpace=BCS_BoneSpace
    alphaBoolBlend {blendOption=Linear}
    boneToModify {boneName=dyn_bangs_02_c}
    rotation=<rot 8.4421 0 0>
   }
   Output Pose(Result=Component To Local(ComponentPose=@n11)) {layerGroup=None}
  ]}
 EventGraph {nodes [
   Event BlueprintUpdateAnimation()
   > n1 = "Gravity Override State Function (Parameter Asset) (Nativized)"(
    CharacterPartAnimInstance=Self-Reference()
    GameTimeInSeconds=GetGameTimeInSeconds()
    Parameters=&/F_MED_TimberStakeSoul_FaceAcc_Params
   )
   > Set Gravity Override(Gravity Override=@n1)
  ]}
]
variables [
 Gravity Override {type=Vector}
 Gravity Override Params {
  type=FortGravityOverrideParameterSet
  default=(Skydive_GoingUp=(GravityOverride=()),Skydive_FloatingDown=(GravityOverride=()),Skydive_DivingDown=(GravityOverride=()),Skydive_Parachuting=(GravityOverride=()),OnGround_Standing=(GravityOverride=()),OnGround_Moving=(GravityOverride=()),OnGround_Moving_HighRotationRate=(GravityOverride=()),OnGround_Crouch_Moving=(GravityOverride=()),Emote_or_Melee=(GravityOverride=()),Swimming=(GravityOverride=()),Front_End=(GravityOverride=()),DBNO=(GravityOverride=()),DBNO_Carried=(GravityOverride=()),OnGround_Falling=(GravityOverride=()),Ziplining=(GravityOverride=()),Skydive_Surfing=(GravityOverride=()),Riding_Standing=(GravityOverride=()),Riding_Moving=(GravityOverride=()),Riding_Moving_Targeting=(GravityOverride=()),Tactical_Sprint=(GravityOverride=()),Rail_Grinding=(GravityOverride=()),)
 }
]
